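# Level Sequence for the Henchman intro cinematic in the Season 12 Battle Pass menu, staged in an Agency-style room behind an elevator transition.
# Elevator door material parameters, a moving frontend root camera and several HM_INT cine cameras with densely keyed motion provide the views.
# Spawn and visibility keys swap set dressing during the sequence, alongside chair animation and a timed hatchet-hit door particle.
# Event tracks trigger on the root camera and map-room light switcher.
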LevelSequence /MiscReferencersOfCosmetics/Animation/Game/MainPlayer/Menu/BR/S12_BP/Henchman/Henchman_Intro
bindings [
 S_Agency_File_2 {
  id=#1
  possessed_class=FortStaticMeshActor
  tracks [{
    type=MovieSceneVisibilityTrack
    sections [{
      type=MovieSceneVisibilitySection
      channels [None {type=MovieSceneScriptingBoolChannel, default=true, keys [[156,false],[488,false],[491,true]]}]
     }]
   }]
 }
 CameraComponent {
  id=#2
  parent_id=#3
  possessed_class=CineCameraComponent
  tracks [
   {
    type=MovieSceneFloatTrack
    display_name="Manual Focus Distance (Focus Settings)"
    sections [{
      type=MovieSceneFloatSection
      channels [None {type=MovieSceneScriptingFloatChannel, default=16.324, keys [[331,16.324],[352.8363,20.7368],[358.8363,93.8866]]}]
     }]
   }
   {
    type=MovieSceneFloatTrack
    display_name=Current Aperture
    sections [{type=MovieSceneFloatSection, channels [None {type=MovieSceneScriptingFloatChannel, default=12.1468}]}]
   }
   {
    type=MovieSceneFloatTrack
    display_name=CurrentFocalLength
    sections [{type=MovieSceneFloatSection, channels [None {type=MovieSceneScriptingFloatChannel, default=30, const [30,1]}]}]
   }
  ]
 }
 S_Agency_File_3 {
  id=#4
  possessed_class=FortStaticMeshActor
  tracks [{
    type=MovieSceneVisibilityTrack
    sections [{type=MovieSceneVisibilitySection, channels [None {type=MovieSceneScriptingBoolChannel, default=false, keys [[113,false],[512.9987,true]]}]}]
   }]
 }
 HM_INT_8000 {
  id=#5
  possessed_class=CineCameraActor
  tracks [
   {
    type=MovieScene3DTransformTrack
    sections [{
      type=MovieScene3DTransformSection
      channels [
       Location.X {type=MovieSceneScriptingDoubleChannel, default=1726.5688, thinned_from=42, keys [[216.9975,1726.5688],[257.9975,1726.6467]]}
       Location.Y {type=MovieSceneScriptingDoubleChannel, default=3309.3247, thinned_from=42, keys [[216.9975,3309.3247],[257.9975,3309.6177]]}
       Location.Z {type=MovieSceneScriptingDoubleChannel, default=870.8049, thinned_from=42, keys [[216.9975,870.8049],[257.9975,870.7986]]}
       Rotation.X {type=MovieSceneScriptingDoubleChannel, default=-1.0843, thinned_from=42, keys [[216.9975,358.9157],[257.9975,358.6717]]}
       Rotation.Y {
        type=MovieSceneScriptingDoubleChannel
        default=7.1691
        thinned_from=42
        keys [[216.9975,7.1691],[225.9975,7.1607],[232.9975,7.2208],[250.9975,7.2052],[252.9975,7.2348],[257.9975,7.2123]]
       }
       Rotation.Z {type=MovieSceneScriptingDoubleChannel, default=165.1526, thinned_from=42, keys [[216.9975,-194.8474],[257.9975,-195.0849]]}
       Scale.X {type=MovieSceneScriptingDoubleChannel, default=1, const [1,42]}
       Scale.Y {type=MovieSceneScriptingDoubleChannel, default=1, const [1,42]}
       Scale.Z {type=MovieSceneScriptingDoubleChannel, default=1, const [1,42]}
      ]
     }]
   }
   {type=MovieSceneSpawnTrack, sections [{type=MovieSceneSpawnSection, channels [None {type=MovieSceneScriptingBoolChannel, default=true}]}]}
  ]
 }
 Cine_Briefcase2 {
  id=#6
  possessed_class=SkeletalMeshActor
  tracks [
   {
    type=MovieScene3DTransformTrack
    sections [{
      type=MovieScene3DTransformSection
      channels [
       Location.X {type=MovieSceneScriptingDoubleChannel, default=1814.2133}
       Location.Y {type=MovieSceneScriptingDoubleChannel, default=3334.0359}
       Location.Z {type=MovieSceneScriptingDoubleChannel, default=800.4074}
       Rotation.X {type=MovieSceneScriptingDoubleChannel, default=-89.999}
       Rotation.Y {type=MovieSceneScriptingDoubleChannel, default=-0.0001}
       Rotation.Z {type=MovieSceneScriptingDoubleChannel, default=95.625}
       Scale.X {type=MovieSceneScriptingDoubleChannel, default=1}
       Scale.Y {type=MovieSceneScriptingDoubleChannel, default=1}
       Scale.Z {type=MovieSceneScriptingDoubleChannel, default=1}
      ]
     }]
   }
   {type=MovieSceneSkeletalAnimationTrack, sections []}
   {
    type=MovieSceneVisibilityTrack
    sections [{type=MovieSceneVisibilitySection, channels [None {type=MovieSceneScriptingBoolChannel, default=false, keys [[493,false],[500,true]]}]}]
   }
  ]
 }
 AB_Wall_Thin16 {
  id=#7
  possessed_class=FortStaticMeshActor
  tracks [
   {
    type=MovieScene3DTransformTrack
    sections [{
      type=MovieScene3DTransformSection
      channels [
       Location.X {type=MovieSceneScriptingDoubleChannel, default=1744}
       Location.Y {type=MovieSceneScriptingDoubleChannel, default=4203}
       Location.Z {type=MovieSceneScriptingDoubleChannel, default=1067}
       Rotation.X {type=MovieSceneScriptingDoubleChannel, default=0}
       Rotation.Y {type=MovieSceneScriptingDoubleChannel, default=0}
       Rotation.Z {type=MovieSceneScriptingDoubleChannel, default=-179.9999}
       Scale.X {type=MovieSceneScriptingDoubleChannel, default=0.5903}
       Scale.Y {type=MovieSceneScriptingDoubleChannel, default=0.75}
       Scale.Z {type=MovieSceneScriptingDoubleChannel, default=0.75}
      ]
     }]
   }
   {
    type=MovieSceneSpawnTrack
    sections [{type=MovieSceneSpawnSection, channels [None {type=MovieSceneScriptingBoolChannel, default=false, keys [[56,true],[515.1137,false]]}]}]
   }
  ]
 }
 CameraComponent {
  id=#8
  parent_id=#9
  possessed_class=CameraComponent
  tracks [{
    type=MovieSceneFloatTrack
    display_name=Field Of View
    sections [{
      type=MovieSceneFloatSection
      channels [None {type=MovieSceneScriptingFloatChannel, default=80, keys [[25,80],[55,80],[498,75.0449],[526.115,75]]}]
     }]
   }]
 }
 sm_apollo_fac_pipe_fitting_71 {
  id=#10
  possessed_class=FortStaticMeshActor
  tracks [
   {
    type=MovieScene3DTransformTrack
    sections [{
      type=MovieScene3DTransformSection
      channels [
       Location.X {type=MovieSceneScriptingDoubleChannel, default=1572.2279}
       Location.Y {type=MovieSceneScriptingDoubleChannel, default=4155.8047}
       Location.Z {type=MovieSceneScriptingDoubleChannel, default=954.0984}
       Rotation.X {type=MovieSceneScriptingDoubleChannel, default=179.9991}
       Rotation.Y {type=MovieSceneScriptingDoubleChannel, default=-0.0005}
       Rotation.Z {type=MovieSceneScriptingDoubleChannel, default=0}
       Scale.X {type=MovieSceneScriptingDoubleChannel, default=0.3932}
       Scale.Y {type=MovieSceneScriptingDoubleChannel, default=0.3932}
       Scale.Z {type=MovieSceneScriptingDoubleChannel, default=0.4126}
      ]
     }]
   }
   {
    type=MovieSceneSpawnTrack
    sections [{type=MovieSceneSpawnSection, channels [None {type=MovieSceneScriptingBoolChannel, default=true, keys [[56,true],[57,true],[512,false]]}]}]
   }
  ]
 }
 Camera_SkullDude_Room {
  id=#11
  possessed_class=Frontend_EventLevel_Camera_C
  tracks [{
    type=MovieScene3DTransformTrack
    sections [{
      type=MovieScene3DTransformSection
      channels [
       Location.X {type=MovieSceneScriptingDoubleChannel, default=2043.6157}
       Location.Y {type=MovieSceneScriptingDoubleChannel, default=3752.7812}
       Location.Z {type=MovieSceneScriptingDoubleChannel, default=931.1245}
       Rotation.X {type=MovieSceneScriptingDoubleChannel, default=0}
       Rotation.Y {type=MovieSceneScriptingDoubleChannel, default=-2}
       Rotation.Z {type=MovieSceneScriptingDoubleChannel, default=-150}
       Scale.X {type=MovieSceneScriptingDoubleChannel, default=1}
       Scale.Y {type=MovieSceneScriptingDoubleChannel, default=1}
       Scale.Z {type=MovieSceneScriptingDoubleChannel, default=1}
      ]
     }]
   }]
 }
 P_Spybase_Hatchet_Hit_Door {
  id=#12
  possessed_class=Emitter
  tracks [
   {
    type=MovieSceneSpawnTrack
    sections [{type=MovieSceneSpawnSection, channels [None {type=MovieSceneScriptingBoolChannel, default=false, keys [[448,false],[450,true]]}]}]
   }
   {
    type=MovieScene3DTransformTrack
    sections [{
      type=MovieScene3DTransformSection
      start_frame=449
      channels [
       Location.X {type=MovieSceneScriptingDoubleChannel, default=1814.6086, const [1816.618,1]}
       Location.Y {type=MovieSceneScriptingDoubleChannel, default=4174.77, const [4173.3813,1]}
       Location.Z {type=MovieSceneScriptingDoubleChannel, default=888.652, const [888.652,1]}
       Rotation.X {type=MovieSceneScriptingDoubleChannel, default=0, const [0,1]}
       Rotation.Y {type=MovieSceneScriptingDoubleChannel, default=0, const [0,1]}
       Rotation.Z {type=MovieSceneScriptingDoubleChannel, default=-90, const [-103.0438,1]}
       Scale.X {type=MovieSceneScriptingDoubleChannel, default=1, const [1,1]}
       Scale.Y {type=MovieSceneScriptingDoubleChannel, default=1, const [1,1]}
       Scale.Z {type=MovieSceneScriptingDoubleChannel, default=1, const [1,1]}
      ]
     }]
   }
   {
    type=MovieSceneParticleTrack
    display_name=Particle System
    sections [{
      type=MovieSceneParticleSection
      channels [None {type=MovieSceneScriptingParticleChannel, keys [[450, "<ParticleKey.ACTIVATE: 0>"], [467.8363, "<ParticleKey.DEACTIVATE: 1>"]]}]
     }]
   }
  ]
 }
 Prop_AntiqueDesk_01_StackofBook_5 {
  id=#13
  possessed_class=FortStaticMeshActor
  tracks [
   {
    type=MovieScene3DTransformTrack
    sections [{
      type=MovieScene3DTransformSection
      channels [
       Location.X {type=MovieSceneScriptingDoubleChannel, default=1504.1002}
       Location.Y {type=MovieSceneScriptingDoubleChannel, default=4146.0293}
       Location.Z {type=MovieSceneScriptingDoubleChannel, default=925.1493}
       Rotation.X {type=MovieSceneScriptingDoubleChannel, default=0}
       Rotation.Y {type=MovieSceneScriptingDoubleChannel, default=0}
       Rotation.Z {type=MovieSceneScriptingDoubleChannel, default=-177.1871}
       Scale.X {type=MovieSceneScriptingDoubleChannel, default=1}
       Scale.Y {type=MovieSceneScriptingDoubleChannel, default=1}
       Scale.Z {type=MovieSceneScriptingDoubleChannel, default=1}
      ]
     }]
   }
   {
    type=MovieSceneSpawnTrack
    sections [{type=MovieSceneSpawnSection, channels [None {type=MovieSceneScriptingBoolChannel, default=false, keys [[56,true],[515.1137,false]]}]}]
   }
  ]
 }
 HM_INT_9000 {
  id=#14
  possessed_class=CineCameraActor
  tracks [
   {
    type=MovieScene3DTransformTrack
    sections [{
      type=MovieScene3DTransformSection
      channels [
       Location.X {
        type=MovieSceneScriptingDoubleChannel
        default=1701.0735
        thinned_from=84
        keys [[368.8363,1701.0735],[376.8363,1701.2128],[378.8363,1708.526],[384.8363,1754.5936],[395.8363,1800.9709],[445.8363,1830.8473],[447.8363,1839.9587],[448.8363,1848.8546],[450.8363,1886.2001],[451.8363,1888.2921],[475.8363,1890.3889]]
       }
       Location.Y {
        type=MovieSceneScriptingDoubleChannel
        default=3477.6082
        thinned_from=119
        keys [[368.8363,3477.6082],[376.8363,3532.5002],[377.8363,3543.7239],[379.8363,3616.8235],[380.8363,3663.1794],[383.8363,3734.967],[394.8363,3849.9434],[445.8363,3945.4912],[446.8363,3964.5657],[448.8363,4036.76],[449.8363,4090.281],[450.8363,4157.3457],[517.8363,4164.9214]]
       }
       Location.Z {type=MovieSceneScriptingDoubleChannel, default=898.8502, thinned_from=150, keys [[368.8363,898.8502],[517.8363,893.3115]]}
       Rotation.X {type=MovieSceneScriptingDoubleChannel, default=0, thinned_from=150, keys [[368.8363,0],[517.8363,0]]}
       Rotation.Y {type=MovieSceneScriptingDoubleChannel, default=0, thinned_from=150, keys [[368.8363,0],[517.8363,0]]}
       Rotation.Z {
        type=MovieSceneScriptingDoubleChannel
        default=-99.8512
        thinned_from=150
        keys [[368.8363,-99.8512],[375.8363,-99.9182],[376.8363,-100.0636],[378.8363,-101.0777],[379.8363,-102.7507],[388.8363,-127.2666],[389.8363,-130.1344],[390.8363,-133.7164],[391.8363,-138.462],[393.8363,-148.8767],[394.8363,-153.8422],[395.8363,-158.1773],[397.8363,-165.6102],[398.8363,-168.9476],[399.8363,-171.9436],[400.8363,-174.5317],[401.8363,-176.6453],[402.8363,-178.154],[403.8363,-179.0849],[439.8363,-192.4188],[444.8363,-192.8918],[445.8363,-193.2433],[446.8363,-196.0335],[447.8363,-198.028],[448.8363,-199.054],[517.8363,-199.054]]
       }
       Scale.X {type=MovieSceneScriptingDoubleChannel, default=1, const [1,150]}
       Scale.Y {type=MovieSceneScriptingDoubleChannel, default=1, const [1,150]}
       Scale.Z {type=MovieSceneScriptingDoubleChannel, default=1, const [1,150]}
      ]
     }]
   }
   {type=MovieSceneSpawnTrack, sections [{type=MovieSceneSpawnSection, channels [None {type=MovieSceneScriptingBoolChannel, default=true}]}]}
  ]
 }
 SkeletalMeshComponent0 {
  id=#15
  parent_id=#16
  possessed_class=SkeletalMeshComponent
  tracks [{
    type=MovieSceneFloatTrack
    display_name=Bounds Scale
    sections [{type=MovieSceneFloatSection, channels [None {type=MovieSceneScriptingFloatChannel, default=5000, const [5000,1]}]}]
   }]
 }
 BP_Elevator_Transition3 {
  id=#17
  possessed_class=BP_Elevator_Transition_C
  tracks [{
    type=MovieSceneVisibilityTrack
    sections [{type=MovieSceneVisibilitySection, channels [None {type=MovieSceneScriptingBoolChannel, default=true, const [true,1]}]}]
   }]
 }
 S_Agency_Picture_Frame_05 {
  id=#18
  possessed_class=FortStaticMeshActor
  tracks [
   {
    type=MovieScene3DTransformTrack
    sections [{
      type=MovieScene3DTransformSection
      channels [
       Location.X {type=MovieSceneScriptingDoubleChannel, default=1716}
       Location.Y {type=MovieSceneScriptingDoubleChannel, default=4185}
       Location.Z {type=MovieSceneScriptingDoubleChannel, default=923}
       Rotation.X {type=MovieSceneScriptingDoubleChannel, default=0}
       Rotation.Y {type=MovieSceneScriptingDoubleChannel, default=0}
       Rotation.Z {type=MovieSceneScriptingDoubleChannel, default=-179.9999}
       Scale.X {type=MovieSceneScriptingDoubleChannel, default=0.4682}
       Scale.Y {type=MovieSceneScriptingDoubleChannel, default=0.4682}
       Scale.Z {type=MovieSceneScriptingDoubleChannel, default=0.4682}
      ]
     }]
   }
   {
    type=MovieSceneSpawnTrack
    sections [{type=MovieSceneSpawnSection, channels [None {type=MovieSceneScriptingBoolChannel, default=false, keys [[56,true],[515.1137,false]]}]}]
   }
  ]
 }
 Doors {
  id=#19
  parent_id=#17
  possessed_class=StaticMeshComponent
  tracks [
   {
    type=MovieSceneComponentMaterialTrack
    display_name=Material Element 0
    sections [{type=MovieSceneParameterSection, channels [Open {type=MovieSceneScriptingFloatChannel, default=0, keys [[26,0],[46,1]]}]}]
   }
   {
    type=MovieSceneComponentMaterialTrack
    display_name=Material Element 0
    sections [{
      type=MovieSceneParameterSection
      channels [
       EnableLightpanDown {type=MovieSceneScriptingFloatChannel, default=1, keys [[6,0],[7,1],[22,1],[23,0]]}
       EnableLightpanUp {type=MovieSceneScriptingFloatChannel, default=0, const [0,1]}
       Open {type=MovieSceneScriptingFloatChannel, default=1, keys [[26,0],[46,1]]}
      ]
     }]
   }
  ]
 }
 sm_apollo_fac_pipe_fitting_68 {
  id=#20
  possessed_class=FortStaticMeshActor
  tracks [
   {
    type=MovieScene3DTransformTrack
    sections [{
      type=MovieScene3DTransformSection
      channels [
       Location.X {type=MovieSceneScriptingDoubleChannel, default=1589.2279}
       Location.Y {type=MovieSceneScriptingDoubleChannel, default=4155.8047}
       Location.Z {type=MovieSceneScriptingDoubleChannel, default=1025.0984}
       Rotation.X {type=MovieSceneScriptingDoubleChannel, default=-88.0349}
       Rotation.Y {type=MovieSceneScriptingDoubleChannel, default=-90}
       Rotation.Z {type=MovieSceneScriptingDoubleChannel, default=88.0355}
       Scale.X {type=MovieSceneScriptingDoubleChannel, default=0.3932}
       Scale.Y {type=MovieSceneScriptingDoubleChannel, default=0.3932}
       Scale.Z {type=MovieSceneScriptingDoubleChannel, default=0.4126}
      ]
     }]
   }
   {
    type=MovieSceneSpawnTrack
    sections [{type=MovieSceneSpawnSection, channels [None {type=MovieSceneScriptingBoolChannel, default=false, keys [[56,true],[512,false]]}]}]
   }
  ]
 }
 CameraComponent {
  id=#21
  parent_id=#14
  possessed_class=CineCameraComponent
  tracks [
   {
    type=MovieSceneFloatTrack
    display_name=Current Focal Length
    sections [{type=MovieSceneFloatSection, channels [None {type=MovieSceneScriptingFloatChannel, default=18, const [18,1]}]}]
   }
   {
    type=MovieSceneFloatTrack
    display_name="Manual Focus Distance (Focus Settings)"
    sections [{type=MovieSceneFloatSection, channels [None {type=MovieSceneScriptingFloatChannel}]}]
   }
   {
    type=MovieSceneFloatTrack
    display_name=Current Aperture
    sections [{type=MovieSceneFloatSection, channels [None {type=MovieSceneScriptingFloatChannel, default=2.8}]}]
   }
  ]
 }
 S_Agency_Computer_01 {
  id=#22
  possessed_class=FortStaticMeshActor
  tracks [{
    type=MovieScene3DTransformTrack
    sections [{
      type=MovieScene3DTransformSection
      channels [
       Location.X {type=MovieSceneScriptingDoubleChannel, default=1683.9276, const [1683.9276,1]}
       Location.Y {type=MovieSceneScriptingDoubleChannel, default=3299.8533, const [3299.8533,1]}
       Location.Z {type=MovieSceneScriptingDoubleChannel, default=866.8563, const [866.8563,1]}
       Rotation.X {type=MovieSceneScriptingDoubleChannel, default=0, const [0,1]}
       Rotation.Y {type=MovieSceneScriptingDoubleChannel, default=0, const [0,1]}
       Rotation.Z {type=MovieSceneScriptingDoubleChannel, default=0, const [0,1]}
       Scale.X {type=MovieSceneScriptingDoubleChannel, default=1, const [1,1]}
       Scale.Y {type=MovieSceneScriptingDoubleChannel, default=1, const [1,1]}
       Scale.Z {type=MovieSceneScriptingDoubleChannel, default=1, const [1,1]}
      ]
     }]
   }]
 }
 sm_apollo_fac_pipe_fitting_70 {
  id=#23
  possessed_class=FortStaticMeshActor
  tracks [
   {
    type=MovieScene3DTransformTrack
    sections [{
      type=MovieScene3DTransformSection
      channels [
       Location.X {type=MovieSceneScriptingDoubleChannel, default=1629.2279}
       Location.Y {type=MovieSceneScriptingDoubleChannel, default=4155.8047}
       Location.Z {type=MovieSceneScriptingDoubleChannel, default=987.0984}
       Rotation.X {type=MovieSceneScriptingDoubleChannel, default=179.9991}
       Rotation.Y {type=MovieSceneScriptingDoubleChannel, default=-0.0005}
       Rotation.Z {type=MovieSceneScriptingDoubleChannel, default=0}
       Scale.X {type=MovieSceneScriptingDoubleChannel, default=0.3932}
       Scale.Y {type=MovieSceneScriptingDoubleChannel, default=0.3932}
       Scale.Z {type=MovieSceneScriptingDoubleChannel, default=0.4126}
      ]
     }]
   }
   {
    type=MovieSceneSpawnTrack
    sections [{type=MovieSceneSpawnSection, channels [None {type=MovieSceneScriptingBoolChannel, default=true, keys [[56,true],[57,true],[512,false]]}]}]
   }
  ]
 }
 sm_apollo_fac_pipe_fitting_74 {
  id=#24
  possessed_class=FortStaticMeshActor
  tracks [
   {
    type=MovieScene3DTransformTrack
    sections [{
      type=MovieScene3DTransformSection
      channels [
       Location.X {type=MovieSceneScriptingDoubleChannel, default=1776.2279}
       Location.Y {type=MovieSceneScriptingDoubleChannel, default=4162.8047}
       Location.Z {type=MovieSceneScriptingDoubleChannel, default=1116.0984}
       Rotation.X {type=MovieSceneScriptingDoubleChannel, default=-0.0009}
       Rotation.Y {type=MovieSceneScriptingDoubleChannel, default=0.0003}
       Rotation.Z {type=MovieSceneScriptingDoubleChannel, default=180}
       Scale.X {type=MovieSceneScriptingDoubleChannel, default=0.3932}
       Scale.Y {type=MovieSceneScriptingDoubleChannel, default=0.3932}
       Scale.Z {type=MovieSceneScriptingDoubleChannel, default=0.4126}
      ]
     }]
   }
   {
    type=MovieSceneSpawnTrack
    sections [{type=MovieSceneSpawnSection, channels [None {type=MovieSceneScriptingBoolChannel, default=true, keys [[56,true],[57,true],[512,false]]}]}]
   }
  ]
 }
 Neo_Door_03 {
  id=#25
  possessed_class=FortStaticMeshActor
  tracks [
   {
    type=MovieScene3DTransformTrack
    sections [{
      type=MovieScene3DTransformSection
      channels [
       Location.X {type=MovieSceneScriptingDoubleChannel, default=1779}
       Location.Y {type=MovieSceneScriptingDoubleChannel, default=4193}
       Location.Z {type=MovieSceneScriptingDoubleChannel, default=779}
       Rotation.X {type=MovieSceneScriptingDoubleChannel, default=0}
       Rotation.Y {type=MovieSceneScriptingDoubleChannel, default=0}
       Rotation.Z {type=MovieSceneScriptingDoubleChannel, default=-0.0001}
       Scale.X {type=MovieSceneScriptingDoubleChannel, default=0.5834}
       Scale.Y {type=MovieSceneScriptingDoubleChannel, default=1}
       Scale.Z {type=MovieSceneScriptingDoubleChannel, default=0.7587}
      ]
     }]
   }
   {
    type=MovieSceneSpawnTrack
    sections [{
      type=MovieSceneSpawnSection
      channels [None {type=MovieSceneScriptingBoolChannel, default=false, keys [[56,true],[367,true],[368,false],[515.1137,false]]}]
     }]
   }
  ]
 }
 sm_apollo_fac_pipe_fitting_69 {
  id=#26
  possessed_class=FortStaticMeshActor
  tracks [
   {
    type=MovieScene3DTransformTrack
    sections [{
      type=MovieScene3DTransformSection
      channels [
       Location.X {type=MovieSceneScriptingDoubleChannel, default=1669.2279}
       Location.Y {type=MovieSceneScriptingDoubleChannel, default=4155.8047}
       Location.Z {type=MovieSceneScriptingDoubleChannel, default=985.0984}
       Rotation.X {type=MovieSceneScriptingDoubleChannel, default=179.9991}
       Rotation.Y {type=MovieSceneScriptingDoubleChannel, default=-0.0005}
       Rotation.Z {type=MovieSceneScriptingDoubleChannel, default=0}
       Scale.X {type=MovieSceneScriptingDoubleChannel, default=0.3932}
       Scale.Y {type=MovieSceneScriptingDoubleChannel, default=0.3932}
       Scale.Z {type=MovieSceneScriptingDoubleChannel, default=0.4126}
      ]
     }]
   }
   {
    type=MovieSceneSpawnTrack
    sections [{
      type=MovieSceneSpawnSection
      channels [None {type=MovieSceneScriptingBoolChannel, default=true, keys [[56,true],[57,true],[515.1137,false]]}]
     }]
   }
  ]
 }
 AB_Wall_Thin14 {
  id=#27
  possessed_class=FortStaticMeshActor
  tracks [
   {
    type=MovieScene3DTransformTrack
    sections [{
      type=MovieScene3DTransformSection
      channels [
       Location.X {type=MovieSceneScriptingDoubleChannel, default=1813}
       Location.Y {type=MovieSceneScriptingDoubleChannel, default=4281}
       Location.Z {type=MovieSceneScriptingDoubleChannel, default=780}
       Rotation.X {type=MovieSceneScriptingDoubleChannel, default=0}
       Rotation.Y {type=MovieSceneScriptingDoubleChannel, default=0}
       Rotation.Z {type=MovieSceneScriptingDoubleChannel, default=-154.6874}
       Scale.X {type=MovieSceneScriptingDoubleChannel, default=0.5961}
       Scale.Y {type=MovieSceneScriptingDoubleChannel, default=0.7558}
       Scale.Z {type=MovieSceneScriptingDoubleChannel, default=0.7558}
      ]
     }]
   }
   {
    type=MovieSceneSpawnTrack
    sections [{type=MovieSceneSpawnSection, channels [None {type=MovieSceneScriptingBoolChannel, default=false, keys [[56,true],[515.1137,false]]}]}]
   }
  ]
 }
 S_MountainBase_WeightBench_01 {
  id=#28
  possessed_class=FortStaticMeshActor
  tracks [
   {
    type=MovieScene3DTransformTrack
    sections [{
      type=MovieScene3DTransformSection
      channels [
       Location.X {type=MovieSceneScriptingDoubleChannel, default=1966}
       Location.Y {type=MovieSceneScriptingDoubleChannel, default=3421}
       Location.Z {type=MovieSceneScriptingDoubleChannel, default=781}
       Rotation.X {type=MovieSceneScriptingDoubleChannel, default=0}
       Rotation.Y {type=MovieSceneScriptingDoubleChannel, default=0}
       Rotation.Z {type=MovieSceneScriptingDoubleChannel, default=-84.375}
       Scale.X {type=MovieSceneScriptingDoubleChannel, default=1}
       Scale.Y {type=MovieSceneScriptingDoubleChannel, default=1}
       Scale.Z {type=MovieSceneScriptingDoubleChannel, default=1}
      ]
     }]
   }
   {
    type=MovieSceneVisibilityTrack
    sections [{type=MovieSceneVisibilitySection, channels [None {type=MovieSceneScriptingBoolChannel, default=false}]}]
   }
  ]
 }
 S_MountainBase_WeightBench_2 {
  id=#29
  possessed_class=FortStaticMeshActor
  tracks [
   {
    type=MovieScene3DTransformTrack
    sections [{
      type=MovieScene3DTransformSection
      channels [
       Location.X {type=MovieSceneScriptingDoubleChannel, default=1966}
       Location.Y {type=MovieSceneScriptingDoubleChannel, default=3858}
       Location.Z {type=MovieSceneScriptingDoubleChannel, default=781}
       Rotation.X {type=MovieSceneScriptingDoubleChannel, default=0}
       Rotation.Y {type=MovieSceneScriptingDoubleChannel, default=0}
       Rotation.Z {type=MovieSceneScriptingDoubleChannel, default=-84.375}
       Scale.X {type=MovieSceneScriptingDoubleChannel, default=1}
       Scale.Y {type=MovieSceneScriptingDoubleChannel, default=1}
       Scale.Z {type=MovieSceneScriptingDoubleChannel, default=1}
      ]
     }]
   }
   {
    type=MovieSceneSpawnTrack
    sections [{type=MovieSceneSpawnSection, channels [None {type=MovieSceneScriptingBoolChannel, default=false, keys [[56,true],[515.1137,false]]}]}]
   }
   {
    type=MovieSceneVisibilityTrack
    sections [{type=MovieSceneVisibilitySection, channels [None {type=MovieSceneScriptingBoolChannel, default=false, const [false,1]}]}]
   }
  ]
 }
 HM_INT_2000 {
  id=#30
  possessed_class=CineCameraActor
  tracks [
   {type=MovieSceneSpawnTrack, sections [{type=MovieSceneSpawnSection, channels [None {type=MovieSceneScriptingBoolChannel, default=true}]}]}
   {
    type=MovieScene3DTransformTrack
    sections [{
      type=MovieScene3DTransformSection
      channels [
       Location.X {type=MovieSceneScriptingDoubleChannel, default=1920.4551, const [1920.4551,51]}
       Location.Y {type=MovieSceneScriptingDoubleChannel, default=3423.6504, const [3423.6504,51]}
       Location.Z {type=MovieSceneScriptingDoubleChannel, default=793.6766, const [793.6766,51]}
       Rotation.X {type=MovieSceneScriptingDoubleChannel, default=-0.0278, thinned_from=51, keys [[116,-360.0278],[166,-360.2153]]}
       Rotation.Y {
        type=MovieSceneScriptingDoubleChannel
        default=4.4096
        thinned_from=51
        keys [[116,4.4096],[122,4.4159],[125,4.4023],[126,4.3873],[128,4.3251],[130,4.3064],[136,4.3114],[140,4.2507],[141,4.2279],[142,4.2187],[145,4.2235],[148,4.2844],[150,4.2926],[152,4.2776],[156,4.3273],[161,4.3176],[166,4.3687]]
       }
       Rotation.Z {type=MovieSceneScriptingDoubleChannel, default=-145.259, thinned_from=51, keys [[116,-145.259],[166,-145.1325]]}
       Scale.X {type=MovieSceneScriptingDoubleChannel, default=1, const [1,51]}
       Scale.Y {type=MovieSceneScriptingDoubleChannel, default=1, const [1,51]}
       Scale.Z {type=MovieSceneScriptingDoubleChannel, default=1, const [1,51]}
      ]
     }]
   }
  ]
 }
 Cine_Chair {
  id=#31
  possessed_class=SkeletalMeshActor
  tracks [
   {
    type=MovieScene3DTransformTrack
    sections [{
      type=MovieScene3DTransformSection
      channels [
       Location.X {type=MovieSceneScriptingDoubleChannel, default=0}
       Location.Y {type=MovieSceneScriptingDoubleChannel, default=0}
       Location.Z {type=MovieSceneScriptingDoubleChannel, default=0}
       Rotation.X {type=MovieSceneScriptingDoubleChannel, default=0}
       Rotation.Y {type=MovieSceneScriptingDoubleChannel, default=0}
       Rotation.Z {type=MovieSceneScriptingDoubleChannel, default=90}
       Scale.X {type=MovieSceneScriptingDoubleChannel, default=1}
       Scale.Y {type=MovieSceneScriptingDoubleChannel, default=1}
       Scale.Z {type=MovieSceneScriptingDoubleChannel, default=1}
      ]
     }]
   }
   {
    type=MovieSceneSkeletalAnimationTrack
    sections [
     {type=MovieSceneSkeletalAnimationSection, end_frame=196, start_frame=167}
     {type=MovieSceneSkeletalAnimationSection, end_frame=218, start_frame=195}
     {type=MovieSceneSkeletalAnimationSection, end_frame=497, start_frame=368}
    ]
   }
   {type=MovieSceneSpawnTrack, sections [{type=MovieSceneSpawnSection, channels [None {type=MovieSceneScriptingBoolChannel, default=true}]}]}
  ]
 }
 paper_map {
  id=#32
  possessed_class=FortStaticMeshActor
  tracks [
   {
    type=MovieScene3DTransformTrack
    sections [{
      type=MovieScene3DTransformSection
      channels [
       Location.X {type=MovieSceneScriptingDoubleChannel, default=1667.7213}
       Location.Y {type=MovieSceneScriptingDoubleChannel, default=3337.3113}
       Location.Z {type=MovieSceneScriptingDoubleChannel, default=874.0111}
       Rotation.X {type=MovieSceneScriptingDoubleChannel, default=0}
       Rotation.Y {type=MovieSceneScriptingDoubleChannel, default=0}
       Rotation.Z {type=MovieSceneScriptingDoubleChannel, default=85}
       Scale.X {type=MovieSceneScriptingDoubleChannel, default=1}
       Scale.Y {type=MovieSceneScriptingDoubleChannel, default=1}
       Scale.Z {type=MovieSceneScriptingDoubleChannel, default=1}
      ]
     }]
   }
   {
    type=MovieSceneVisibilityTrack
    sections [{
      type=MovieSceneVisibilitySection
      channels [None {type=MovieSceneScriptingBoolChannel, default=false, keys [[234.5025,false],[500.115,true]]}]
     }]
   }
  ]
 }
 S_Agency_Chair_2 {
  id=#33
  possessed_class=FortStaticMeshActor
  tracks [
   {
    type=MovieScene3DTransformTrack
    sections [{
      type=MovieScene3DTransformSection
      channels [
       Location.X {type=MovieSceneScriptingDoubleChannel, default=1694.4733}
       Location.Y {type=MovieSceneScriptingDoubleChannel, default=3420.8901}
       Location.Z {type=MovieSceneScriptingDoubleChannel, default=781.2214}
       Rotation.X {type=MovieSceneScriptingDoubleChannel, default=0}
       Rotation.Y {type=MovieSceneScriptingDoubleChannel, default=0}
       Rotation.Z {type=MovieSceneScriptingDoubleChannel, default=-133.5063}
       Scale.X {type=MovieSceneScriptingDoubleChannel, default=1}
       Scale.Y {type=MovieSceneScriptingDoubleChannel, default=1}
       Scale.Z {type=MovieSceneScriptingDoubleChannel, default=1}
      ]
     }]
   }
   {
    type=MovieSceneVisibilityTrack
    sections [{
      type=MovieSceneVisibilitySection
      channels [None {type=MovieSceneScriptingBoolChannel, default=true, keys [[381,false],[441,false],[443,true]]}]
     }]
   }
  ]
 }
 S_Agency_Picture_Frame_6 {
  id=#34
  possessed_class=FortStaticMeshActor
  tracks [
   {
    type=MovieScene3DTransformTrack
    sections [{
      type=MovieScene3DTransformSection
      channels [
       Location.X {type=MovieSceneScriptingDoubleChannel, default=1958}
       Location.Y {type=MovieSceneScriptingDoubleChannel, default=4185}
       Location.Z {type=MovieSceneScriptingDoubleChannel, default=922}
       Rotation.X {type=MovieSceneScriptingDoubleChannel, default=0}
       Rotation.Y {type=MovieSceneScriptingDoubleChannel, default=0}
       Rotation.Z {type=MovieSceneScriptingDoubleChannel, default=-179.9999}
       Scale.X {type=MovieSceneScriptingDoubleChannel, default=0.4682}
       Scale.Y {type=MovieSceneScriptingDoubleChannel, default=0.4682}
       Scale.Z {type=MovieSceneScriptingDoubleChannel, default=0.4682}
      ]
     }]
   }
   {
    type=MovieSceneSpawnTrack
    sections [{type=MovieSceneSpawnSection, channels [None {type=MovieSceneScriptingBoolChannel, default=false, keys [[56,true],[515.1137,false]]}]}]
   }
  ]
 }
 BP_Battlepass_LightSwitcher_Maproom {
  id=#35
  possessed_class=BP_Battlepass_LightSwitcher_C
  tracks [{
    type=MovieSceneEventTrack
    sections [{
      type=MovieSceneEventTriggerSection
      channels [None {type=MovieSceneScriptingEventChannel, const ["<Struct 'MovieSceneEvent' (0x8516af03980) {}>", 1]}]
     }]
   }]
 }
 S_Agency_FileCabinet_3 {
  id=#36
  possessed_class=FortStaticMeshActor
  tracks [
   {
    type=MovieScene3DTransformTrack
    sections [{
      type=MovieScene3DTransformSection
      channels [
       Location.X {type=MovieSceneScriptingDoubleChannel, default=1508.2341}
       Location.Y {type=MovieSceneScriptingDoubleChannel, default=4155.2788}
       Location.Z {type=MovieSceneScriptingDoubleChannel, default=780}
       Rotation.X {type=MovieSceneScriptingDoubleChannel, default=0}
       Rotation.Y {type=MovieSceneScriptingDoubleChannel, default=0}
       Rotation.Z {type=MovieSceneScriptingDoubleChannel, default=177.8127}
       Scale.X {type=MovieSceneScriptingDoubleChannel, default=1}
       Scale.Y {type=MovieSceneScriptingDoubleChannel, default=1}
       Scale.Z {type=MovieSceneScriptingDoubleChannel, default=1}
      ]
     }]
   }
   {
    type=MovieSceneSpawnTrack
    sections [{type=MovieSceneSpawnSection, channels [None {type=MovieSceneScriptingBoolChannel, default=false, keys [[56,true],[515.1137,false]]}]}]
   }
  ]
 }
 BP_Frontend_EventLevel_NavObject_Mannequin3 {
  id=#37
  possessed_class=BP_Frontend_EventLevel_NavObject_Mannequin_C
  tracks [
   {
    type=MovieSceneVisibilityTrack
    sections [{type=MovieSceneVisibilitySection, channels [None {type=MovieSceneScriptingBoolChannel, default=false, keys [[491,false],[492,true]]}]}]
   }
   {
    type=MovieScene3DTransformTrack
    sections [{
      type=MovieScene3DTransformSection
      channels [
       Location.X {type=MovieSceneScriptingDoubleChannel, default=1658, const [1658,1]}
       Location.Y {type=MovieSceneScriptingDoubleChannel, default=3553, const [3553,1]}
       Location.Z {type=MovieSceneScriptingDoubleChannel, default=780, const [780,1]}
       Rotation.X {type=MovieSceneScriptingDoubleChannel, default=0, const [0,1]}
       Rotation.Y {type=MovieSceneScriptingDoubleChannel, default=0, const [0,1]}
       Rotation.Z {type=MovieSceneScriptingDoubleChannel, default=-100.0002, const [-50,1]}
       Scale.X {type=MovieSceneScriptingDoubleChannel, default=1, const [1,1]}
       Scale.Y {type=MovieSceneScriptingDoubleChannel, default=1, const [1,1]}
       Scale.Z {type=MovieSceneScriptingDoubleChannel, default=1, const [1,1]}
      ]
     }]
   }
  ]
 }
 CameraComponent {
  id=#38
  parent_id=#39
  possessed_class=CineCameraComponent
  tracks [
   {
    type=MovieSceneFloatTrack
    display_name=Current Focal Length
    sections [{type=MovieSceneFloatSection, channels [None {type=MovieSceneScriptingFloatChannel, default=40, const [40,1]}]}]
   }
   {
    type=MovieSceneFloatTrack
    display_name="Manual Focus Distance (Focus Settings)"
    sections [{type=MovieSceneFloatSection, channels [None {type=MovieSceneScriptingFloatChannel, default=77.2012, const [77.2012,1]}]}]
   }
   {
    type=MovieSceneFloatTrack
    display_name=Current Aperture
    sections [{type=MovieSceneFloatSection, channels [None {type=MovieSceneScriptingFloatChannel, default=2.8}]}]
   }
  ]
 }
 Camera_Root_Movement {
  id=#9
  possessed_class=Frontend_EventLevel_Camera_C
  tracks [
   {
    type=MovieScene3DTransformTrack
    sections [{
      type=MovieScene3DTransformSection
      channels [
       Location.X {
        type=MovieSceneScriptingDoubleChannel
        default=2043.6157
        thinned_from=11
        keys [[50,2358],[56,2358],[498,1675.6344],[512,2043.6157],[521,1677.0088],[524,1733.7444],[529,2040.3418],[539.115,2043.6157]]
       }
       Location.Y {
        type=MovieSceneScriptingDoubleChannel
        default=3752.7812
        thinned_from=11
        keys [[50,3984],[56,3984],[498,3564.3757],[512,3752.7812],[521,3565.0793],[524,3594.1279],[529,3751.1052],[539.115,3752.7812]]
       }
       Location.Z {
        type=MovieSceneScriptingDoubleChannel
        default=931.1245
        thinned_from=11
        keys [[50,926],[497.0013,916.7296],[498,935.9061],[512,931.1245],[521,935.8882],[524,935.1509],[529,931.1661],[539.115,931.1245]]
       }
       Rotation.X {type=MovieSceneScriptingDoubleChannel, default=0, const [0,11]}
       Rotation.Y {type=MovieSceneScriptingDoubleChannel, default=-2, thinned_from=11, keys [[50,0],[56,0],[497.0013,-2],[539.115,-2]]}
       Rotation.Z {type=MovieSceneScriptingDoubleChannel, default=-150, thinned_from=11, keys [[50,-130],[56,-130],[498,-150],[539.115,-150]]}
       Scale.X {type=MovieSceneScriptingDoubleChannel, default=1, const [1,6]}
       Scale.Y {type=MovieSceneScriptingDoubleChannel, default=1, const [1,6]}
       Scale.Z {type=MovieSceneScriptingDoubleChannel, default=1, const [1,6]}
      ]
     }]
   }
   {
    type=MovieSceneEventTrack
    display_name=Event Stream In
    sections [{
      type=MovieSceneEventTriggerSection
      channels [None {type=MovieSceneScriptingEventChannel, const ["<Struct 'MovieSceneEvent' (0x8516af80d20) {}>", 1]}]
     }]
   }
  ]
 }
 HM_INT_4000 {
  id=#3
  possessed_class=CineCameraActor
  tracks [
   {type=MovieSceneSpawnTrack, sections [{type=MovieSceneSpawnSection, channels [None {type=MovieSceneScriptingBoolChannel, default=true}]}]}
   {
    type=MovieScene3DTransformTrack
    sections [{
      type=MovieScene3DTransformSection
      channels [
       Location.X {type=MovieSceneScriptingDoubleChannel, default=1700.0762, thinned_from=47, keys [[324.8363,1700.0762],[370.8363,1700.1533]]}
       Location.Y {type=MovieSceneScriptingDoubleChannel, default=3391.9709, thinned_from=3, keys [[324.8363,3391.9709],[370.8363,3391.9722]]}
       Location.Z {type=MovieSceneScriptingDoubleChannel, default=917.4921, thinned_from=3, keys [[324.8363,917.4921],[371.8363,917.4919]]}
       Rotation.X {type=MovieSceneScriptingDoubleChannel, default=-0.8157, thinned_from=48, keys [[324.8363,359.1843],[371.8363,359.1895]]}
       Rotation.Y {
        type=MovieSceneScriptingDoubleChannel
        default=0.1043
        thinned_from=48
        keys [[324.8363,0.1043],[325.8363,0.1022],[326.8363,0.0989],[327.8363,0.0973],[328.8363,0.0982],[329.8363,0.1091],[330.8363,0.1274],[331.8363,0.1421],[332.8363,0.1518],[334.8363,0.164],[335.8363,0.1739],[336.8363,0.1861],[337.8363,0.2016],[338.8363,0.221],[339.8363,0.2307],[340.8363,0.2351],[341.8363,0.2445],[342.8363,0.2695],[343.8363,0.305],[344.8363,0.3226],[345.8363,0.3127],[346.8363,0.2991],[347.8363,0.2987],[348.8363,0.3124],[349.8363,0.3333],[350.8363,0.3437],[351.8363,0.3464],[352.8363,0.3395],[353.8363,0.3199],[354.8363,0.2932],[355.8363,0.2591],[356.8363,0.2285],[357.8363,0.1984],[358.8363,0.1659],[359.8363,0.1433],[360.8363,0.1416],[361.8363,0.1656],[362.8363,0.1938],[363.8363,0.207],[364.8363,0.2029],[365.8363,0.1824],[366.8363,0.1451],[367.8363,0.0833],[368.8363,0.0003],[369.8363,-0.0851],[370.8363,-0.141],[371.8363,-0.1349]]
       }
       Rotation.Z {type=MovieSceneScriptingDoubleChannel, default=94.2351, thinned_from=48, keys [[324.8363,94.2351],[371.8363,94.0797]]}
       Scale.X {type=MovieSceneScriptingDoubleChannel, default=1, const [1,48]}
       Scale.Y {type=MovieSceneScriptingDoubleChannel, default=1, const [1,48]}
       Scale.Z {type=MovieSceneScriptingDoubleChannel, default=1, const [1,48]}
      ]
     }]
   }
  ]
 }
 AB_Wall_Thin5 {
  id=#40
  possessed_class=FortStaticMeshActor
  tracks [
   {
    type=MovieScene3DTransformTrack
    sections [{
      type=MovieScene3DTransformSection
      channels [
       Location.X {type=MovieSceneScriptingDoubleChannel, default=1744}
       Location.Y {type=MovieSceneScriptingDoubleChannel, default=4195}
       Location.Z {type=MovieSceneScriptingDoubleChannel, default=780}
       Rotation.X {type=MovieSceneScriptingDoubleChannel, default=0}
       Rotation.Y {type=MovieSceneScriptingDoubleChannel, default=0}
       Rotation.Z {type=MovieSceneScriptingDoubleChannel, default=-179.9999}
       Scale.X {type=MovieSceneScriptingDoubleChannel, default=0.5903}
       Scale.Y {type=MovieSceneScriptingDoubleChannel, default=0.75}
       Scale.Z {type=MovieSceneScriptingDoubleChannel, default=0.75}
      ]
     }]
   }
   {
    type=MovieSceneSpawnTrack
    sections [{type=MovieSceneSpawnSection, channels [None {type=MovieSceneScriptingBoolChannel, default=false, keys [[56,true],[515.1137,false]]}]}]
   }
  ]
 }
 "FortStaticMeshActor (2)" {
  id=#41
  possessed_class=FortStaticMeshActor
  tracks [{
    type=MovieSceneVisibilityTrack
    sections [{type=MovieSceneVisibilitySection, channels [None {type=MovieSceneScriptingBoolChannel, default=false, keys [[146,false],[512.9987,true]]}]}]
   }]
 }
 Cine_Axe_Gadget_Cine_Henchman_Axe {
  id=#42
  possessed_class=SkeletalMeshActor
  tracks [
   {
    type=MovieScene3DTransformTrack
    sections [{
      type=MovieScene3DTransformSection
      channels [
       Location.X {type=MovieSceneScriptingDoubleChannel, default=0}
       Location.Y {type=MovieSceneScriptingDoubleChannel, default=0}
       Location.Z {type=MovieSceneScriptingDoubleChannel, default=0}
       Rotation.X {type=MovieSceneScriptingDoubleChannel, default=0}
       Rotation.Y {type=MovieSceneScriptingDoubleChannel, default=0}
       Rotation.Z {type=MovieSceneScriptingDoubleChannel, default=90}
       Scale.X {type=MovieSceneScriptingDoubleChannel, default=1}
       Scale.Y {type=MovieSceneScriptingDoubleChannel, default=1}
       Scale.Z {type=MovieSceneScriptingDoubleChannel, default=1}
      ]
     }]
   }
   {type=MovieSceneSkeletalAnimationTrack, sections [{type=MovieSceneSkeletalAnimationSection, end_frame=516, start_frame=368}]}
   {type=MovieSceneSkeletalAnimationTrack, sections [{type=MovieSceneSkeletalAnimationSection, end_frame=258, start_frame=224}]}
   {type=MovieSceneSpawnTrack, sections [{type=MovieSceneSpawnSection, channels [None {type=MovieSceneScriptingBoolChannel, default=true}]}]}
  ]
 }
 AB_Wall_Thin15 {
  id=#43
  possessed_class=FortStaticMeshActor
  tracks [
   {
    type=MovieScene3DTransformTrack
    sections [{
      type=MovieScene3DTransformSection
      channels [
       Location.X {type=MovieSceneScriptingDoubleChannel, default=1439}
       Location.Y {type=MovieSceneScriptingDoubleChannel, default=4203}
       Location.Z {type=MovieSceneScriptingDoubleChannel, default=1067}
       Rotation.X {type=MovieSceneScriptingDoubleChannel, default=0}
       Rotation.Y {type=MovieSceneScriptingDoubleChannel, default=0}
       Rotation.Z {type=MovieSceneScriptingDoubleChannel, default=-179.9999}
       Scale.X {type=MovieSceneScriptingDoubleChannel, default=0.5903}
       Scale.Y {type=MovieSceneScriptingDoubleChannel, default=0.75}
       Scale.Z {type=MovieSceneScriptingDoubleChannel, default=0.75}
      ]
     }]
   }
   {
    type=MovieSceneSpawnTrack
    sections [{type=MovieSceneSpawnSection, channels [None {type=MovieSceneScriptingBoolChannel, default=false, keys [[56,true],[515.1137,false]]}]}]
   }
  ]
 }
 S_Agency_FileCabinet_8 {
  id=#44
  possessed_class=FortStaticMeshActor
  tracks [
   {
    type=MovieScene3DTransformTrack
    sections [{
      type=MovieScene3DTransformSection
      channels [
       Location.X {type=MovieSceneScriptingDoubleChannel, default=1440.3167}
       Location.Y {type=MovieSceneScriptingDoubleChannel, default=4151.9414}
       Location.Z {type=MovieSceneScriptingDoubleChannel, default=780}
       Rotation.X {type=MovieSceneScriptingDoubleChannel, default=0}
       Rotation.Y {type=MovieSceneScriptingDoubleChannel, default=0}
       Rotation.Z {type=MovieSceneScriptingDoubleChannel, default=-177.1871}
       Scale.X {type=MovieSceneScriptingDoubleChannel, default=1}
       Scale.Y {type=MovieSceneScriptingDoubleChannel, default=1}
       Scale.Z {type=MovieSceneScriptingDoubleChannel, default=1}
      ]
     }]
   }
   {
    type=MovieSceneSpawnTrack
    sections [{type=MovieSceneSpawnSection, channels [None {type=MovieSceneScriptingBoolChannel, default=false, keys [[56,true],[515.1137,false]]}]}]
   }
  ]
 }
 HM_INT_3000 {
  id=#45
  possessed_class=CineCameraActor
  tracks [
   {
    type=MovieScene3DTransformTrack
    sections [{
      type=MovieScene3DTransformSection
      channels [
       Location.X {type=MovieSceneScriptingDoubleChannel, default=1724.4628, thinned_from=65, keys [[258,1724.4628],[322,1733.6051]]}
       Location.Y {type=MovieSceneScriptingDoubleChannel, default=3294.6738, thinned_from=65, keys [[258,3294.6738],[322,3295.4778]]}
       Location.Z {type=MovieSceneScriptingDoubleChannel, default=896.1303, thinned_from=65, keys [[258,896.1303],[322,895.0007]]}
       Rotation.X {
        type=MovieSceneScriptingDoubleChannel
        default=4.199
        thinned_from=65
        keys [[258,4.199],[259,4.2004],[260,4.2255],[262,4.242],[264,4.2416],[268,4.2692],[273,4.24],[275,4.215],[277,4.2146],[279,4.2335],[284,4.2516],[289,4.2419],[298,4.2712],[300,4.2979],[301,4.2995],[302,4.2898],[304,4.2055],[305,4.195],[306,4.202],[307,4.2176],[309,4.2788],[310,4.2943],[314,4.3134],[322,4.3792]]
       }
       Rotation.Y {
        type=MovieSceneScriptingDoubleChannel
        default=7.9887
        thinned_from=65
        keys [[258,7.9887],[259,7.9888],[261,8.0225],[263,8.0229],[268,7.9195],[277,7.915],[280,7.9953],[283,8.008],[284,7.9967],[287,7.9396],[292,7.9294],[294,7.9063],[299,7.951],[303,7.9322],[308,7.9661],[314,7.9396],[318,7.9618],[320,7.9222],[322,7.9136]]
       }
       Rotation.Z {type=MovieSceneScriptingDoubleChannel, default=133.0525, thinned_from=65, keys [[258,493.0525],[322,495.6163]]}
       Scale.X {type=MovieSceneScriptingDoubleChannel, default=1, const [1,65]}
       Scale.Y {type=MovieSceneScriptingDoubleChannel, default=1, const [1,65]}
       Scale.Z {type=MovieSceneScriptingDoubleChannel, default=1, const [1,65]}
      ]
     }]
   }
   {type=MovieSceneSpawnTrack, sections [{type=MovieSceneSpawnSection, channels [None {type=MovieSceneScriptingBoolChannel, default=true}]}]}
  ]
 }
 sm_apollo_fac_pipe_fitting_66 {
  id=#46
  possessed_class=FortStaticMeshActor
  tracks [
   {
    type=MovieScene3DTransformTrack
    sections [{
      type=MovieScene3DTransformSection
      channels [
       Location.X {type=MovieSceneScriptingDoubleChannel, default=1545.2279}
       Location.Y {type=MovieSceneScriptingDoubleChannel, default=4155.8047}
       Location.Z {type=MovieSceneScriptingDoubleChannel, default=978.0984}
       Rotation.X {type=MovieSceneScriptingDoubleChannel, default=88.0343}
       Rotation.Y {type=MovieSceneScriptingDoubleChannel, default=-90}
       Rotation.Z {type=MovieSceneScriptingDoubleChannel, default=-88.0345}
       Scale.X {type=MovieSceneScriptingDoubleChannel, default=0.3932}
       Scale.Y {type=MovieSceneScriptingDoubleChannel, default=0.3932}
       Scale.Z {type=MovieSceneScriptingDoubleChannel, default=0.4126}
      ]
     }]
   }
   {
    type=MovieSceneSpawnTrack
    sections [{type=MovieSceneSpawnSection, channels [None {type=MovieSceneScriptingBoolChannel, default=false, keys [[56,true],[515.1137,false]]}]}]
   }
  ]
 }
 S_Agency_Security_Camera_01 {
  id=#47
  possessed_class=FortStaticMeshActor
  tracks [
   {
    type=MovieScene3DTransformTrack
    sections [{
      type=MovieScene3DTransformSection
      channels [
       Location.X {type=MovieSceneScriptingDoubleChannel, default=1632}
       Location.Y {type=MovieSceneScriptingDoubleChannel, default=3320}
       Location.Z {type=MovieSceneScriptingDoubleChannel, default=875.9183}
       Rotation.X {type=MovieSceneScriptingDoubleChannel, default=-63.9014}
       Rotation.Y {type=MovieSceneScriptingDoubleChannel, default=-4.8224}
       Rotation.Z {type=MovieSceneScriptingDoubleChannel, default=-109.1464}
       Scale.X {type=MovieSceneScriptingDoubleChannel, default=0.8474}
       Scale.Y {type=MovieSceneScriptingDoubleChannel, default=0.8474}
       Scale.Z {type=MovieSceneScriptingDoubleChannel, default=0.8474}
      ]
     }]
   }
   {
    type=MovieSceneVisibilityTrack
    sections [{
      type=MovieSceneVisibilitySection
      channels [None {type=MovieSceneScriptingBoolChannel, default=false, keys [[234.5025,false],[512.9987,true]]}]
     }]
   }
  ]
 }
 Cine_Door_01 {
  id=#48
  possessed_class=SkeletalMeshActor
  tracks [
   {
    type=MovieScene3DTransformTrack
    sections [{
      type=MovieScene3DTransformSection
      channels [
       Location.X {type=MovieSceneScriptingDoubleChannel, default=0}
       Location.Y {type=MovieSceneScriptingDoubleChannel, default=0}
       Location.Z {type=MovieSceneScriptingDoubleChannel, default=0}
       Rotation.X {type=MovieSceneScriptingDoubleChannel, default=0}
       Rotation.Y {type=MovieSceneScriptingDoubleChannel, default=0}
       Rotation.Z {type=MovieSceneScriptingDoubleChannel, default=0}
       Scale.X {type=MovieSceneScriptingDoubleChannel, default=1}
       Scale.Y {type=MovieSceneScriptingDoubleChannel, default=1}
       Scale.Z {type=MovieSceneScriptingDoubleChannel, default=1}
      ]
     }]
   }
   {type=MovieSceneSkeletalAnimationTrack, sections []}
   {type=MovieSceneSkeletalAnimationTrack, sections [{type=MovieSceneSkeletalAnimationSection, end_frame=517, start_frame=368}]}
   {type=MovieSceneSpawnTrack, sections [{type=MovieSceneSpawnSection, channels [None {type=MovieSceneScriptingBoolChannel, default=true}]}]}
  ]
 }
 sm_apollo_fac_pipe_fitting_72 {
  id=#49
  possessed_class=FortStaticMeshActor
  tracks [
   {
    type=MovieScene3DTransformTrack
    sections [{
      type=MovieScene3DTransformSection
      channels [
       Location.X {type=MovieSceneScriptingDoubleChannel, default=2129.2271}
       Location.Y {type=MovieSceneScriptingDoubleChannel, default=4162.8047}
       Location.Z {type=MovieSceneScriptingDoubleChannel, default=1041.0973}
       Rotation.X {type=MovieSceneScriptingDoubleChannel, default=179.8407}
       Rotation.Y {type=MovieSceneScriptingDoubleChannel, default=-90}
       Rotation.Z {type=MovieSceneScriptingDoubleChannel, default=0.1584}
       Scale.X {type=MovieSceneScriptingDoubleChannel, default=0.3932}
       Scale.Y {type=MovieSceneScriptingDoubleChannel, default=0.3932}
       Scale.Z {type=MovieSceneScriptingDoubleChannel, default=0.473}
      ]
     }]
   }
   {
    type=MovieSceneSpawnTrack
    sections [{type=MovieSceneSpawnSection, channels [None {type=MovieSceneScriptingBoolChannel, default=true, keys [[56,true],[57,true],[512,false]]}]}]
   }
  ]
 }
 BP12_SkullDude_Mannequin {
  id=#16
  possessed_class=BP12_SkullDude_Mannequin_C
  tracks [
   {
    type=MovieScene3DTransformTrack
    sections [{
      type=MovieScene3DTransformSection
      channels [
       Location.X {type=MovieSceneScriptingDoubleChannel, default=0, const [0,1]}
       Location.Y {type=MovieSceneScriptingDoubleChannel, default=0, const [0,1]}
       Location.Z {type=MovieSceneScriptingDoubleChannel, default=0, const [0,1]}
       Rotation.X {type=MovieSceneScriptingDoubleChannel, default=0, const [0,1]}
       Rotation.Y {type=MovieSceneScriptingDoubleChannel, default=0, const [0,1]}
       Rotation.Z {type=MovieSceneScriptingDoubleChannel, default=0, const [0,1]}
       Scale.X {type=MovieSceneScriptingDoubleChannel, default=1, const [1,1]}
       Scale.Y {type=MovieSceneScriptingDoubleChannel, default=1, const [1,1]}
       Scale.Z {type=MovieSceneScriptingDoubleChannel, default=1, const [1,1]}
      ]
     }]
   }
   {
    type=MovieSceneSkeletalAnimationTrack
    sections [
     {type=MovieSceneSkeletalAnimationSection, end_frame=116, start_frame=56}
     {type=MovieSceneSkeletalAnimationSection, end_frame=167, start_frame=116}
     {type=MovieSceneSkeletalAnimationSection, end_frame=322, start_frame=257}
     {type=MovieSceneSkeletalAnimationSection, end_frame=196, start_frame=167}
     {type=MovieSceneSkeletalAnimationSection, end_frame=217, start_frame=196}
     {type=MovieSceneSkeletalAnimationSection, end_frame=258, start_frame=217}
     {type=MovieSceneSkeletalAnimationSection, end_frame=512, start_frame=368}
     {type=MovieSceneSkeletalAnimationSection, end_frame=368, start_frame=322}
    ]
   }
   {
    type=MovieSceneVisibilityTrack
    sections [{
      type=MovieSceneVisibilitySection
      channels [None {type=MovieSceneScriptingBoolChannel, default=false, keys [[45,false],[57,true],[512.9987,false]]}]
     }]
   }
  ]
 }
 CameraComponent {
  id=#50
  parent_id=#5
  possessed_class=CineCameraComponent
  tracks [
   {
    type=MovieSceneFloatTrack
    display_name=Current Focal Length
    sections [{type=MovieSceneFloatSection, channels [None {type=MovieSceneScriptingFloatChannel, default=25, const [25,1]}]}]
   }
   {
    type=MovieSceneFloatTrack
    display_name="Manual Focus Distance (Focus Settings)"
    sections [{type=MovieSceneFloatSection, channels [None {type=MovieSceneScriptingFloatChannel, const [31.1963,2]}]}]
   }
   {
    type=MovieSceneFloatTrack
    display_name=Current Aperture
    sections [{type=MovieSceneFloatSection, channels [None {type=MovieSceneScriptingFloatChannel, default=2.8}]}]
   }
  ]
 }
 HM_INT_6000 {
  id=#51
  possessed_class=CineCameraActor
  tracks [
   {
    type=MovieScene3DTransformTrack
    sections [{
      type=MovieScene3DTransformSection
      channels [
       Location.X {type=MovieSceneScriptingDoubleChannel, default=1587.7306, const [1587.7306,29]}
       Location.Y {type=MovieSceneScriptingDoubleChannel, default=3134.5793, const [3134.5793,29]}
       Location.Z {type=MovieSceneScriptingDoubleChannel, default=991.2724, const [991.2724,29]}
       Rotation.X {type=MovieSceneScriptingDoubleChannel, default=1.5748, thinned_from=29, keys [[196.4937,361.5748],[224.4937,361.713]]}
       Rotation.Y {type=MovieSceneScriptingDoubleChannel, default=-33.0017, thinned_from=29, keys [[196.4937,-33.0017],[224.4937,-32.9507]]}
       Rotation.Z {type=MovieSceneScriptingDoubleChannel, default=77.5818, thinned_from=29, keys [[196.4937,77.5818],[224.4937,77.6541]]}
       Scale.X {type=MovieSceneScriptingDoubleChannel, default=1, const [1,29]}
       Scale.Y {type=MovieSceneScriptingDoubleChannel, default=1, const [1,29]}
       Scale.Z {type=MovieSceneScriptingDoubleChannel, default=1, const [1,29]}
      ]
     }]
   }
   {type=MovieSceneSpawnTrack, sections [{type=MovieSceneSpawnSection, channels [None {type=MovieSceneScriptingBoolChannel, default=true}]}]}
  ]
 }
 AB_Wall_Thin17 {
  id=#52
  possessed_class=FortStaticMeshActor
  tracks [
   {
    type=MovieScene3DTransformTrack
    sections [{
      type=MovieScene3DTransformSection
      channels [
       Location.X {type=MovieSceneScriptingDoubleChannel, default=2051}
       Location.Y {type=MovieSceneScriptingDoubleChannel, default=4203}
       Location.Z {type=MovieSceneScriptingDoubleChannel, default=1067}
       Rotation.X {type=MovieSceneScriptingDoubleChannel, default=0}
       Rotation.Y {type=MovieSceneScriptingDoubleChannel, default=0}
       Rotation.Z {type=MovieSceneScriptingDoubleChannel, default=-179.9999}
       Scale.X {type=MovieSceneScriptingDoubleChannel, default=0.5903}
       Scale.Y {type=MovieSceneScriptingDoubleChannel, default=0.75}
       Scale.Z {type=MovieSceneScriptingDoubleChannel, default=0.75}
      ]
     }]
   }
   {
    type=MovieSceneSpawnTrack
    sections [{type=MovieSceneSpawnSection, channels [None {type=MovieSceneScriptingBoolChannel, default=false, keys [[56,true],[515.1137,false]]}]}]
   }
  ]
 }
 S_MountainBase_Poster_5 {
  id=#53
  possessed_class=FortStaticMeshActor
  tracks [
   {
    type=MovieScene3DTransformTrack
    sections [{
      type=MovieScene3DTransformSection
      channels [
       Location.X {type=MovieSceneScriptingDoubleChannel, default=1821.7601}
       Location.Y {type=MovieSceneScriptingDoubleChannel, default=4176.6113}
       Location.Z {type=MovieSceneScriptingDoubleChannel, default=895.3779}
       Rotation.X {type=MovieSceneScriptingDoubleChannel, default=0}
       Rotation.Y {type=MovieSceneScriptingDoubleChannel, default=0}
       Rotation.Z {type=MovieSceneScriptingDoubleChannel, default=-179.9999}
       Scale.X {type=MovieSceneScriptingDoubleChannel, default=0.7298}
       Scale.Y {type=MovieSceneScriptingDoubleChannel, default=0.7298}
       Scale.Z {type=MovieSceneScriptingDoubleChannel, default=0.7298}
      ]
     }]
   }
   {
    type=MovieSceneSpawnTrack
    sections [{type=MovieSceneSpawnSection, channels [None {type=MovieSceneScriptingBoolChannel, default=false, keys [[56,true],[401,false]]}]}]
   }
  ]
 }
 CameraComponent {
  id=#54
  parent_id=#45
  possessed_class=CineCameraComponent
  tracks [
   {
    type=MovieSceneFloatTrack
    display_name=Current Focal Length
    sections [{type=MovieSceneFloatSection, channels [None {type=MovieSceneScriptingFloatChannel, default=45, const [45,1]}]}]
   }
   {
    type=MovieSceneFloatTrack
    display_name="Manual Focus Distance (Focus Settings)"
    sections [{type=MovieSceneFloatSection, channels [None {type=MovieSceneScriptingFloatChannel, default=95.9056, const [95.9056,1]}]}]
   }
   {
    type=MovieSceneFloatTrack
    display_name=Current Aperture
    sections [{type=MovieSceneFloatSection, channels [None {type=MovieSceneScriptingFloatChannel, default=2.8}]}]
   }
  ]
 }
 BP_Battlepass_LightSwitcher_Hench {
  id=#55
  possessed_class=BP_Battlepass_LightSwitcher_C
  tracks [{
    type=MovieSceneEventTrack
    sections [{
      type=MovieSceneEventTriggerSection
      channels [None {
        type=MovieSceneScriptingEventChannel
        keys [[57, "<Struct 'MovieSceneEvent' (0x8516af81ea0) {}>"], [511.9987, "<Struct 'MovieSceneEvent' (0x8516af82080) {}>"]]
       }]
     }]
   }]
 }
 CameraComponent {
  id=#56
  parent_id=#51
  possessed_class=CineCameraComponent
  tracks [
   {
    type=MovieSceneFloatTrack
    display_name=Current Focal Length
    sections [{type=MovieSceneFloatSection, channels [None {type=MovieSceneScriptingFloatChannel, default=60, const [60,1]}]}]
   }
   {
    type=MovieSceneFloatTrack
    display_name="Manual Focus Distance (Focus Settings)"
    sections [{type=MovieSceneFloatSection, channels [None {type=MovieSceneScriptingFloatChannel}]}]
   }
   {
    type=MovieSceneFloatTrack
    display_name=Current Aperture
    sections [{type=MovieSceneFloatSection, channels [None {type=MovieSceneScriptingFloatChannel, default=2.8}]}]
   }
  ]
 }
 CameraComponent {
  id=#57
  parent_id=#11
  possessed_class=CameraComponent
  tracks [{
    type=MovieSceneFloatTrack
    display_name=Field Of View
    sections [{type=MovieSceneFloatSection, channels [None {type=MovieSceneScriptingFloatChannel, default=75}]}]
   }]
 }
 CameraComponent {
  id=#58
  parent_id=#59
  possessed_class=CineCameraComponent
  tracks [
   {
    type=MovieSceneFloatTrack
    display_name=Current Focal Length
    sections [{type=MovieSceneFloatSection, channels [None {type=MovieSceneScriptingFloatChannel, const [90,58]}]}]
   }
   {
    type=MovieSceneFloatTrack
    display_name="Manual Focus Distance (Focus Settings)"
    sections [{type=MovieSceneFloatSection, channels [None {type=MovieSceneScriptingFloatChannel, default=261.3523, const [261.3523,1]}]}]
   }
   {
    type=MovieSceneFloatTrack
    display_name=Current Aperture
    sections [{type=MovieSceneFloatSection, channels [None {type=MovieSceneScriptingFloatChannel, default=2.8}]}]
   }
  ]
 }
 S_Agency_File_01 {
  id=#60
  possessed_class=FortStaticMeshActor
  tracks [
   {
    type=MovieScene3DTransformTrack
    sections [{
      type=MovieScene3DTransformSection
      channels [
       Location.X {type=MovieSceneScriptingDoubleChannel, default=1721.276}
       Location.Y {type=MovieSceneScriptingDoubleChannel, default=3354.1511}
       Location.Z {type=MovieSceneScriptingDoubleChannel, default=866.8563}
       Rotation.X {type=MovieSceneScriptingDoubleChannel, default=0}
       Rotation.Y {type=MovieSceneScriptingDoubleChannel, default=0}
       Rotation.Z {type=MovieSceneScriptingDoubleChannel, default=15.0001}
       Scale.X {type=MovieSceneScriptingDoubleChannel, default=1}
       Scale.Y {type=MovieSceneScriptingDoubleChannel, default=1}
       Scale.Z {type=MovieSceneScriptingDoubleChannel, default=1}
      ]
     }]
   }
   {
    type=MovieSceneVisibilityTrack
    sections [{type=MovieSceneVisibilitySection, channels [None {type=MovieSceneScriptingBoolChannel, default=false, const [false,1]}]}]
   }
  ]
 }
 S_Military_GunCase_10 {
  id=#61
  possessed_class=FortStaticMeshActor
  tracks [
   {
    type=MovieScene3DTransformTrack
    sections [{
      type=MovieScene3DTransformSection
      channels [
       Location.X {type=MovieSceneScriptingDoubleChannel, default=1721.7855}
       Location.Y {type=MovieSceneScriptingDoubleChannel, default=4140.6021}
       Location.Z {type=MovieSceneScriptingDoubleChannel, default=857.7849}
       Rotation.X {type=MovieSceneScriptingDoubleChannel, default=0}
       Rotation.Y {type=MovieSceneScriptingDoubleChannel, default=0}
       Rotation.Z {type=MovieSceneScriptingDoubleChannel, default=-159.0626}
       Scale.X {type=MovieSceneScriptingDoubleChannel, default=1}
       Scale.Y {type=MovieSceneScriptingDoubleChannel, default=1}
       Scale.Z {type=MovieSceneScriptingDoubleChannel, default=1}
      ]
     }]
   }
   {
    type=MovieSceneSpawnTrack
    sections [{type=MovieSceneSpawnSection, channels [None {type=MovieSceneScriptingBoolChannel, default=false, keys [[56,true],[515.1137,false]]}]}]
   }
  ]
 }
 HM_INT_5000 {
  id=#59
  possessed_class=CineCameraActor
  tracks [
   {
    type=MovieScene3DTransformTrack
    sections [{
      type=MovieScene3DTransformSection
      channels [
       Location.X {type=MovieSceneScriptingDoubleChannel, default=1786.2262, const [1786.2262,58]}
       Location.Y {type=MovieSceneScriptingDoubleChannel, default=3607.6272, const [3607.6272,58]}
       Location.Z {type=MovieSceneScriptingDoubleChannel, default=890.4482, const [890.4482,58]}
       Rotation.X {type=MovieSceneScriptingDoubleChannel, default=-0.2016, thinned_from=58, keys [[166.0013,359.7984],[225.5038,359.7154]]}
       Rotation.Y {
        type=MovieSceneScriptingDoubleChannel
        default=3.0675
        thinned_from=58
        keys [[166.0013,3.0675],[175.0013,3.0468],[178.0013,3.0577],[182.0013,3.0968],[185.0013,3.1182],[187.0013,3.1419],[189.0013,3.1518],[196.0013,3.1209],[196.1162,3.0922],[197.0013,3.1122],[197.6087,3.0884],[198.0013,3.1023],[198.6087,3.0834],[199.0013,3.0969],[199.6087,3.0783],[200.0013,3.0945],[200.6087,3.0741],[205.6087,3.0466],[207.6087,3.0518],[214.6087,3.1241],[216.6087,3.1509],[217.6087,3.1554],[225.5038,3.1554]]
       }
       Rotation.Z {type=MovieSceneScriptingDoubleChannel, default=-113.5341, thinned_from=58, keys [[166.0013,-473.5341],[225.5038,-473.5253]]}
       Scale.X {type=MovieSceneScriptingDoubleChannel, default=1, const [1,58]}
       Scale.Y {type=MovieSceneScriptingDoubleChannel, default=1, const [1,58]}
       Scale.Z {type=MovieSceneScriptingDoubleChannel, default=1, const [1,58]}
      ]
     }]
   }
   {type=MovieSceneSpawnTrack, sections [{type=MovieSceneSpawnSection, channels [None {type=MovieSceneScriptingBoolChannel, default=true}]}]}
  ]
 }
 Cine_Briefcase {
  id=#62
  possessed_class=SkeletalMeshActor
  tracks [
   {
    type=MovieScene3DTransformTrack
    sections [{
      type=MovieScene3DTransformSection
      channels [
       Location.X {type=MovieSceneScriptingDoubleChannel, default=0}
       Location.Y {type=MovieSceneScriptingDoubleChannel, default=0}
       Location.Z {type=MovieSceneScriptingDoubleChannel, default=0}
       Rotation.X {type=MovieSceneScriptingDoubleChannel, default=0}
       Rotation.Y {type=MovieSceneScriptingDoubleChannel, default=0}
       Rotation.Z {type=MovieSceneScriptingDoubleChannel, default=90}
       Scale.X {type=MovieSceneScriptingDoubleChannel, default=1}
       Scale.Y {type=MovieSceneScriptingDoubleChannel, default=1}
       Scale.Z {type=MovieSceneScriptingDoubleChannel, default=1}
      ]
     }]
   }
   {type=MovieSceneSkeletalAnimationTrack, sections [{type=MovieSceneSkeletalAnimationSection, end_frame=167, start_frame=115}]}
   {type=MovieSceneSpawnTrack, sections [{type=MovieSceneSpawnSection, channels [None {type=MovieSceneScriptingBoolChannel, default=true}]}]}
  ]
 }
 sm_apollo_fac_pipe_fitting_75 {
  id=#63
  possessed_class=FortStaticMeshActor
  tracks [
   {
    type=MovieScene3DTransformTrack
    sections [{
      type=MovieScene3DTransformSection
      channels [
       Location.X {type=MovieSceneScriptingDoubleChannel, default=1889.2271}
       Location.Y {type=MovieSceneScriptingDoubleChannel, default=4162.8047}
       Location.Z {type=MovieSceneScriptingDoubleChannel, default=1041.0973}
       Rotation.X {type=MovieSceneScriptingDoubleChannel, default=89.999}
       Rotation.Y {type=MovieSceneScriptingDoubleChannel, default=-64.6873}
       Rotation.Z {type=MovieSceneScriptingDoubleChannel, default=90}
       Scale.X {type=MovieSceneScriptingDoubleChannel, default=0.3932}
       Scale.Y {type=MovieSceneScriptingDoubleChannel, default=0.3932}
       Scale.Z {type=MovieSceneScriptingDoubleChannel, default=0.4126}
      ]
     }]
   }
   {
    type=MovieSceneSpawnTrack
    sections [{type=MovieSceneSpawnSection, channels [None {type=MovieSceneScriptingBoolChannel, default=false, keys [[56,true],[512,false]]}]}]
   }
  ]
 }
 Cine_Hammer {
  id=#64
  possessed_class=SkeletalMeshActor
  tracks [
   {
    type=MovieScene3DTransformTrack
    sections [{
      type=MovieScene3DTransformSection
      channels [
       Location.X {type=MovieSceneScriptingDoubleChannel, default=0}
       Location.Y {type=MovieSceneScriptingDoubleChannel, default=0}
       Location.Z {type=MovieSceneScriptingDoubleChannel, default=0}
       Rotation.X {type=MovieSceneScriptingDoubleChannel, default=0}
       Rotation.Y {type=MovieSceneScriptingDoubleChannel, default=0}
       Rotation.Z {type=MovieSceneScriptingDoubleChannel, default=90}
       Scale.X {type=MovieSceneScriptingDoubleChannel, default=1}
       Scale.Y {type=MovieSceneScriptingDoubleChannel, default=1}
       Scale.Z {type=MovieSceneScriptingDoubleChannel, default=1}
      ]
     }]
   }
   {
    type=MovieSceneSkeletalAnimationTrack
    sections [
     {type=MovieSceneSkeletalAnimationSection, end_frame=196, start_frame=166}
     {type=MovieSceneSkeletalAnimationSection, end_frame=217, start_frame=196}
    ]
   }
   {type=MovieSceneSkeletalAnimationTrack, sections [{type=MovieSceneSkeletalAnimationSection, end_frame=267, start_frame=224}]}
   {type=MovieSceneSpawnTrack, sections [{type=MovieSceneSpawnSection, channels [None {type=MovieSceneScriptingBoolChannel, default=true}]}]}
  ]
 }
 CameraComponent {
  id=#65
  parent_id=#30
  possessed_class=CineCameraComponent
  tracks [
   {
    type=MovieSceneFloatTrack
    display_name=Current Focal Length
    sections [{type=MovieSceneFloatSection, channels [None {type=MovieSceneScriptingFloatChannel, default=25, const [25,1]}]}]
   }
   {
    type=MovieSceneFloatTrack
    display_name="Manual Focus Distance (Focus Settings)"
    sections [{type=MovieSceneFloatSection, channels [None {type=MovieSceneScriptingFloatChannel, default=135.3644, const [135.3644,1]}]}]
   }
   {
    type=MovieSceneFloatTrack
    display_name=Current Aperture
    sections [{type=MovieSceneFloatSection, channels [None {type=MovieSceneScriptingFloatChannel, default=1.2}]}]
   }
  ]
 }
 sm_apollo_fac_pipe_fitting_73 {
  id=#66
  possessed_class=FortStaticMeshActor
  tracks [
   {
    type=MovieScene3DTransformTrack
    sections [{
      type=MovieScene3DTransformSection
      channels [
       Location.X {type=MovieSceneScriptingDoubleChannel, default=1856.2273}
       Location.Y {type=MovieSceneScriptingDoubleChannel, default=4162.8047}
       Location.Z {type=MovieSceneScriptingDoubleChannel, default=1041.0981}
       Rotation.X {type=MovieSceneScriptingDoubleChannel, default=175.7877}
       Rotation.Y {type=MovieSceneScriptingDoubleChannel, default=90}
       Rotation.Z {type=MovieSceneScriptingDoubleChannel, default=175.7873}
       Scale.X {type=MovieSceneScriptingDoubleChannel, default=0.3932}
       Scale.Y {type=MovieSceneScriptingDoubleChannel, default=0.3932}
       Scale.Z {type=MovieSceneScriptingDoubleChannel, default=0.4126}
      ]
     }]
   }
   {
    type=MovieSceneSpawnTrack
    sections [{type=MovieSceneSpawnSection, channels [None {type=MovieSceneScriptingBoolChannel, default=true, keys [[56,true],[57,true],[512,false]]}]}]
   }
  ]
 }
 AB_Wall_Thin13 {
  id=#67
  possessed_class=FortStaticMeshActor
  tracks [
   {
    type=MovieScene3DTransformTrack
    sections [{
      type=MovieScene3DTransformSection
      channels [
       Location.X {type=MovieSceneScriptingDoubleChannel, default=2048}
       Location.Y {type=MovieSceneScriptingDoubleChannel, default=4203}
       Location.Z {type=MovieSceneScriptingDoubleChannel, default=780}
       Rotation.X {type=MovieSceneScriptingDoubleChannel, default=0}
       Rotation.Y {type=MovieSceneScriptingDoubleChannel, default=0}
       Rotation.Z {type=MovieSceneScriptingDoubleChannel, default=-179.9999}
       Scale.X {type=MovieSceneScriptingDoubleChannel, default=0.5903}
       Scale.Y {type=MovieSceneScriptingDoubleChannel, default=0.75}
       Scale.Z {type=MovieSceneScriptingDoubleChannel, default=0.75}
      ]
     }]
   }
   {
    type=MovieSceneSpawnTrack
    sections [{type=MovieSceneSpawnSection, channels [None {type=MovieSceneScriptingBoolChannel, default=false, keys [[56,true],[515.1137,false]]}]}]
   }
  ]
 }
 HM_INT_1000 {
  id=#39
  possessed_class=CineCameraActor
  tracks [
   {
    type=MovieScene3DTransformTrack
    sections [{
      type=MovieScene3DTransformSection
      channels [
       Location.X {type=MovieSceneScriptingDoubleChannel, default=1913.2089, thinned_from=61, keys [[56,1913.2089],[116,1921.6434]]}
       Location.Y {type=MovieSceneScriptingDoubleChannel, default=3749.2288, thinned_from=61, keys [[56,3749.2288],[116,3652.8455]]}
       Location.Z {type=MovieSceneScriptingDoubleChannel, default=786.4686, thinned_from=61, keys [[56,786.4686],[116,786.5552]]}
       Rotation.X {type=MovieSceneScriptingDoubleChannel, default=0.4155, thinned_from=61, keys [[56,360.4155],[116,360.501]]}
       Rotation.Y {
        type=MovieSceneScriptingDoubleChannel
        default=2.7218
        thinned_from=61
        keys [[56,2.7218],[58,2.7113],[59,2.715],[65,2.7838],[69,2.8179],[72,2.8601],[73,2.8646],[74,2.8516],[75,2.82],[77,2.736],[78,2.705],[79,2.6858],[81,2.671],[86,2.6681],[88,2.6803],[89,2.6928],[90,2.7122],[91,2.7426],[92,2.7797],[93,2.8047],[94,2.8152],[97,2.8098],[99,2.8024],[100,2.8094],[102,2.8453],[104,2.8985],[105,2.9183],[106,2.9262],[108,2.9241],[109,2.914],[110,2.8957],[112,2.848],[113,2.8347],[114,2.835],[115,2.8444],[116,2.8444]]
       }
       Rotation.Z {type=MovieSceneScriptingDoubleChannel, default=-175.7705, thinned_from=61, keys [[56,184.2295],[116,184.3961]]}
       Scale.X {type=MovieSceneScriptingDoubleChannel, default=1, const [1,61]}
       Scale.Y {type=MovieSceneScriptingDoubleChannel, default=1, const [1,61]}
       Scale.Z {type=MovieSceneScriptingDoubleChannel, default=1, const [1,61]}
      ]
     }]
   }
   {
    type=MovieSceneSpawnTrack
    sections [{type=MovieSceneSpawnSection, channels [None {type=MovieSceneScriptingBoolChannel, default=true, keys [[56,true],[505.9987,false]]}]}]
   }
  ]
 }
 sm_apollo_fac_pipe_fitting_67 {
  id=#68
  possessed_class=FortStaticMeshActor
  tracks [
   {
    type=MovieScene3DTransformTrack
    sections [{
      type=MovieScene3DTransformSection
      channels [
       Location.X {type=MovieSceneScriptingDoubleChannel, default=1589.2279}
       Location.Y {type=MovieSceneScriptingDoubleChannel, default=4155.8047}
       Location.Z {type=MovieSceneScriptingDoubleChannel, default=1060.0984}
       Rotation.X {type=MovieSceneScriptingDoubleChannel, default=-87.3521}
       Rotation.Y {type=MovieSceneScriptingDoubleChannel, default=-90}
       Rotation.Z {type=MovieSceneScriptingDoubleChannel, default=87.3506}
       Scale.X {type=MovieSceneScriptingDoubleChannel, default=0.3932}
       Scale.Y {type=MovieSceneScriptingDoubleChannel, default=0.3932}
       Scale.Z {type=MovieSceneScriptingDoubleChannel, default=0.4126}
      ]
     }]
   }
   {
    type=MovieSceneSpawnTrack
    sections [{type=MovieSceneSpawnSection, channels [None {type=MovieSceneScriptingBoolChannel, default=false, keys [[56,true],[515.1137,false]]}]}]
   }
  ]
 }
 S_Agency_FileCabinet_9 {
  id=#69
  possessed_class=FortStaticMeshActor
  tracks [
   {
    type=MovieScene3DTransformTrack
    sections [{
      type=MovieScene3DTransformSection
      channels [
       Location.X {type=MovieSceneScriptingDoubleChannel, default=1719}
       Location.Y {type=MovieSceneScriptingDoubleChannel, default=4151}
       Location.Z {type=MovieSceneScriptingDoubleChannel, default=780}
       Rotation.X {type=MovieSceneScriptingDoubleChannel, default=0}
       Rotation.Y {type=MovieSceneScriptingDoubleChannel, default=0}
       Rotation.Z {type=MovieSceneScriptingDoubleChannel, default=-180}
       Scale.X {type=MovieSceneScriptingDoubleChannel, default=1}
       Scale.Y {type=MovieSceneScriptingDoubleChannel, default=1}
       Scale.Z {type=MovieSceneScriptingDoubleChannel, default=1}
      ]
     }]
   }
   {
    type=MovieSceneSpawnTrack
    sections [{type=MovieSceneSpawnSection, channels [None {type=MovieSceneScriptingBoolChannel, default=false, keys [[56,true],[515.1137,false]]}]}]
   }
  ]
 }
]
display_rate {denominator=1, numerator=30}
tick_resolution {denominator=1, numerator=24000}
tracks [
 {
  type=MovieSceneFadeTrack
  display_name=FadeTrack
  sections [{type=MovieSceneFadeSection, channels [None {type=MovieSceneScriptingFloatChannel, default=1, keys [[57,1],[63,0],[508,0],[512.9987,1]]}]}]
 }
 {
  type=MovieSceneAudioTrack
  display_name=Audio
  sections [{
    type=MovieSceneAudioSection
    end_frame=557
    start_frame=56
    channels [
     Volume {type=MovieSceneScriptingFloatChannel, default=1}
     Pitch {type=MovieSceneScriptingFloatChannel, default=1}
     AttachActor {type=MovieSceneScriptingActorReferenceChannel, default="<Struct 'MovieSceneObjectBindingID' (0x850fb7d2520) {}>"}
    ]
   }]
 }
 {
  type=MovieSceneCameraCutTrack
  sections [
   {type=MovieSceneCameraCutSection, end_frame=56, start_frame=0}
   {type=MovieSceneCameraCutSection, end_frame=116, start_frame=56}
   {type=MovieSceneCameraCutSection, end_frame=167, start_frame=116}
   {type=MovieSceneCameraCutSection, end_frame=196, start_frame=167}
   {type=MovieSceneCameraCutSection, end_frame=217, start_frame=196}
   {type=MovieSceneCameraCutSection, end_frame=258, start_frame=217}
   {type=MovieSceneCameraCutSection, end_frame=322, start_frame=258}
   {type=MovieSceneCameraCutSection, end_frame=368, start_frame=322}
   {type=MovieSceneCameraCutSection, end_frame=512, start_frame=368}
   {type=MovieSceneCameraCutSection, end_frame=516, start_frame=512}
  ]
 }
]
